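# Ready sound cue for the StreetGoth cosmetic pickaxe, built from its swing waves and shared distant fire-axe swings.
# Each quality branch mixes a pitch-raised StreetGoth layer on close attenuation with a quieter fire-axe layer on medium attenuation.
# A quality-level root scales variation from full random pools down to single waves, with the two lowest-variation slots sharing one mixer.
# The cue primes on load and overrides concurrency to prevent new instances at the limit.
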SoundCue /BRCosmetics/Sounds/PickAxes/StreetGoth/PA_StreetGoth_Ready_Cue
root=QualityLevel_0
nodes [
 {id=Attenuation_0, children [Modulator_5], properties {attenuationSettings=/Game/Sounds/Attenuation/Fort_Weapon_Pickaxe_Close_Att}}
 {id=Attenuation_1, children [Random_1], properties {attenuationSettings=/Game/Sounds/Attenuation/Fort_Weapon_Pickaxe_Medium_Att}}
 {id=Attenuation_2, children [Modulator_1], properties {attenuationSettings=&Fort_Weapon_Pickaxe_Close_Att}}
 {id=Attenuation_4, children [Modulator_2], properties {attenuationSettings=&Fort_Weapon_Pickaxe_Close_Att}}
 {id=Attenuation_8, children [Random_6], properties {attenuationSettings=&Fort_Weapon_Pickaxe_Medium_Att}}
 {id=Attenuation_9, children [WavePlayer_44], properties {attenuationSettings=&Fort_Weapon_Pickaxe_Medium_Att}}
 {id=Mixer_0, children [Attenuation_0, Modulator_8], properties {inputVolume [1,1]}}
 {id=Mixer_1, children [Attenuation_2, Modulator_7], properties {inputVolume [1.1,1]}}
 {id=Mixer_2, children [Attenuation_4, Modulator_0], properties {inputVolume [1.1,1]}}
 {id=Modulator_0, children [Attenuation_1], properties {pitchMax=1.1, pitchMin=1, volumeMax=0.35, volumeMin=0.35}}
 {id=Modulator_1, children [Random_0], properties {pitchMax=1.15, pitchMin=1.1}}
 {id=Modulator_2, children [Random_2], properties {pitchMax=1.15, pitchMin=1.1}}
 {id=Modulator_5, children [WavePlayer_18], properties {pitchMax=1.15, pitchMin=1.1}}
 {id=Modulator_7, children [Attenuation_8], properties {pitchMax=1.1, pitchMin=1, volumeMax=0.35, volumeMin=0.35}}
 {id=Modulator_8, children [Attenuation_9], properties {pitchMax=1.1, pitchMin=1, volumeMax=0.35, volumeMin=0.35}}
 {id=QualityLevel_0, children [Mixer_1,Mixer_2,Mixer_0,Mixer_0]}
 {
  id=Random_0
  children [WavePlayer_6, WavePlayer_7, WavePlayer_8, WavePlayer_9, WavePlayer_10, WavePlayer_11]
  properties {weights [1,1,1,1,1,1]}
 }
 {id=Random_1, children [WavePlayer_5,WavePlayer_4,WavePlayer_3,WavePlayer_0,WavePlayer_1,WavePlayer_2], properties {weights [1,1,1,1,1,1]}}
 {id=Random_2, children [WavePlayer_14, WavePlayer_15, WavePlayer_17], properties {weights [1,1,1]}}
 {
  id=Random_6
  children [WavePlayer_43, WavePlayer_42, WavePlayer_41, WavePlayer_31, WavePlayer_38, WavePlayer_40]
  properties {weights [1,1,1,1,1,1]}
 }
 {id=WavePlayer_0, properties {soundWaveAssetPtr=/Game/Athena/Sounds/Weapons/Fire_Axe_Swing_Distant_03}}
 {id=WavePlayer_1, properties {soundWaveAssetPtr=/Game/Athena/Sounds/Weapons/Fire_Axe_Swing_Distant_05}}
 {id=WavePlayer_10, properties {soundWaveAssetPtr=&/PA_StreetGoth_Swing_05}}
 {id=WavePlayer_11, properties {soundWaveAssetPtr=&/PA_StreetGoth_Swing_06}}
 {id=WavePlayer_14, properties {soundWaveAssetPtr=&/PA_StreetGoth_Swing_01}}
 {id=WavePlayer_15, properties {soundWaveAssetPtr=&/PA_StreetGoth_Swing_02}}
 {id=WavePlayer_17, properties {soundWaveAssetPtr=&/PA_StreetGoth_Swing_03}}
 {id=WavePlayer_18, properties {soundWaveAssetPtr=&PA_StreetGoth_Swing_01}}
 {id=WavePlayer_2, properties {soundWaveAssetPtr=/Game/Athena/Sounds/Weapons/Fire_Axe_Swing_Distant_06}}
 {id=WavePlayer_3, properties {soundWaveAssetPtr=/Game/Athena/Sounds/Weapons/Fire_Axe_Swing_Distant_02}}
 {id=WavePlayer_31, properties {soundWaveAssetPtr=&Fire_Axe_Swing_Distant_03}}
 {id=WavePlayer_38, properties {soundWaveAssetPtr=&Fire_Axe_Swing_Distant_05}}
 {id=WavePlayer_4, properties {soundWaveAssetPtr=/Game/Athena/Sounds/Weapons/Fire_Axe_Swing_Distant_01}}
 {id=WavePlayer_40, properties {soundWaveAssetPtr=&Fire_Axe_Swing_Distant_06}}
 {id=WavePlayer_41, properties {soundWaveAssetPtr=&Fire_Axe_Swing_Distant_02}}
 {id=WavePlayer_42, properties {soundWaveAssetPtr=&Fire_Axe_Swing_Distant_01}}
 {id=WavePlayer_43, properties {soundWaveAssetPtr=/Game/Athena/Sounds/Weapons/Fire_Axe_Swing_Distant_04}}
 {id=WavePlayer_44, properties {soundWaveAssetPtr=&Fire_Axe_Swing_Distant_01}}
 {id=WavePlayer_5, properties {soundWaveAssetPtr=&Fire_Axe_Swing_Distant_04}}
 {id=WavePlayer_6, properties {soundWaveAssetPtr=&PA_StreetGoth_Swing_01}}
 {id=WavePlayer_7, properties {soundWaveAssetPtr=&PA_StreetGoth_Swing_02}}
 {id=WavePlayer_8, properties {soundWaveAssetPtr=&PA_StreetGoth_Swing_03}}
 {id=WavePlayer_9, properties {soundWaveAssetPtr=&/PA_StreetGoth_Swing_04}}
]
properties {
 attenuationSettings=&Fort_Weapon_Projectile_Close_Athena_Att
 audioPropertiesSheet=/BRCosmetics/Sounds/Settings/PropertySheets/Pickaxe/APS_Cosmetic_Pickaxe_Ready
 bOverrideConcurrency=true
 bPrimeOnLoad=true
 maxDistance=4950
 soundClassObject=&Fort_WeaponMelee
 volumeMultiplier=0.08
 concurrencyOverrides {resolutionRule=PreventNew}
}
